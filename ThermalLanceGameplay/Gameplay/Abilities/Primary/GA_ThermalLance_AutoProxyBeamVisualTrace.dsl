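; Thermal Lance ability that updates the weapon's Niagara proxy beam visual while it fires, for smooth display on the local client.
; Firing-tag listeners start and stop it, deferring the stop one tick if the tag clears on the frame it was added.
; An immediate targeting trace shows the beam, then async traces reschedule each tick, independent of attack rate.
; Hit results, or a synthesized max-range endpoint when nothing is hit, drive the beam.
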(event Ability|EventActivateAbility
  (bind _weaponfiringgameplaytag (Variables|Settings|GetWeaponFiringGameplayTag))
  ;[0.01086 0.2066 0.04136 0.5] ActivateAbility: Setup delegates that are invoked when the laser weapon starts and stops firing.
  (bind _asynctaskproxy (Ability|Tasks|WaitGameplayTagAdd _weaponfiringgameplaytag)
    (:then
      (Variables|Transient|SetAsyncTask_RangeFiringAdd _asynctaskproxy)
      (bind _asynctaskproxy_2 (Ability|Tasks|WaitGameplayTagRemove _weaponfiringgameplaytag)
        (:then
          (Variables|Transient|SetAsyncTask_RangeFiringRemove _asynctaskproxy_2))
        (:Removed
          (if (== (Utilities|GetFrameCount) (Variables|Transient|GetWeaponFiringEnableFrameCount))
            (bind _returnvalue (Utilities|Time|SetTimerforNextTickbyEvent (EventDispatchers|CreateEvent :!event SetWeaponAttackingToDisabled)))
            (Variables|Transient|SetDisableWeaponFiringNextFrameTimerHandle _returnvalue)
            (else
              (CallFunction|SetWeaponAttackingEnabled false))))))
    (:Added
      (Variables|Transient|SetWeaponFiringEnableFrameCount (Utilities|GetFrameCount))
      (Utilities|Time|ClearandInvalidateTimerbyHandle (Variables|Transient|GetDisableWeaponFiringNextFrameTimerHandle))
      (CallFunction|SetWeaponAttackingEnabled true))))

(event Custom|SetWeaponAttackingEnabled
  (params (bEnable :type Boolean))
  ;[0.1478 0.04488 0.15 0.5] Invoked when the laser weapon starts and stops firing.
  (bind _output_get (Variables|Transient|SetbFiring bEnable))
  (if _output_get
    ;;[0.01939 0.1247 0.15 0.5] From next tick, start the async trace to update the beam visuals.
    (bind _returnvalue (Utilities|Time|SetTimerforNextTickbyEvent (EventDispatchers|CreateEvent :!event Custom|PerformBeamVisualTrace)))
    (Variables|Transient|SetPerformBeamVisualTraceDelayTimerHandle _returnvalue)
    ;;[0.15 0.1352 0.08125 0.5] Weapon has just started firing, we need the laser visual to be updated ASAP. Run a single sync trace for this.
    (bind (_asynctaskproxy _targetdata) (Ability|Tasks|PerformTargetingAbilityTask "SyncPerformBeamVisualTrace" "/ThermalLanceGameplay/Gameplay/Targeting/TP_ThermalLance_Targeting.TP_ThermalLance_Targeting" false true)
      (:then
        (Variables|Transient|SetAsyncTask_Targeting _asynctaskproxy))
      (:Targeted
        (CallFunction|ParseTargetDataandUpdateNiagaraBeam _targetdata)))
    (else
      ;; Player has stopped firing, handle tasks and niagara effect on the laser.
      (Utilities|Time|ClearandInvalidateTimerbyHandle (Variables|Transient|GetPerformBeamVisualTraceDelayTimerHandle))
      (GameplayTasks|EndTask (Variables|Transient|GetAsyncTask_Targeting))
      (bind _weapon (Utilities|GetThermalLanceWeapon)
        (:Valid
          (Beam|SetAutoProxyBeamNiagaraEnable _weapon false))))))

(event Custom|PerformBeamVisualTrace
  ;[0.01147 0.02026 0.15 0.5] PerformBeamVisualTrace: On the local client, it is important for the laser to update as often as possible to depict a smooth representation of the beam even if the attack rate of the laser is low enough that it isn't update very frame.
  (bind (_asynctaskproxy _targetdata) (Ability|Tasks|PerformTargetingAbilityTask "PerformBeamVisualTrace" "/ThermalLanceGameplay/Gameplay/Targeting/TP_ThermalLance_Targeting.TP_ThermalLance_Targeting" false true true)
    (:then
      (Variables|Transient|SetAsyncTask_Targeting _asynctaskproxy))
    (:Targeted
      (if (Variables|Transient|GetbFiring)
        (bind _returnvalue (Utilities|Time|SetTimerforNextTickbyEvent (delegate Custom|PerformBeamVisualTrace)))
        (Variables|Transient|SetPerformBeamVisualTraceDelayTimerHandle _returnvalue)
        (CallFunction|ParseTargetDataandUpdateNiagaraBeam _targetdata)))))

(event Custom|ParseTargetDataAndUpdateNiagaraBeam
  (params (TargetData :type Gameplay%20Ability%20Target%20Data%20Handle%20Structure%20%28by%20ref%29))
  ;[0.01855 0.04034 0.15 0.5] ParseTargetDataAndUpdateNiagaraBeam: Break down the TargetData into a meaningful HitResult to use for the beam.
  (bind (_bsuccess _hitresult) (Default|ParseTargetData TargetData))
  (if _bsuccess
    ;; IF the trace hit, update the beam info accordingly.
    ;; Enable the beam niagara effect component.
    (bind _weapon (Utilities|GetThermalLanceWeapon)
      (:Valid
        (Beam|SetAutoProxyBeamInfo _weapon _hitresult)
        (Beam|SetAutoProxyBeamNiagaraEnable _weapon true)))))

(event Ability|EventOnEndAbility
  (params (bWasCancelled :type Boolean))
  ;[0.15 0.07379 0.01882 0.5] OnEndAbility: Clean up.
  (Utilities|FlowControl|Sequence
    (:then_0
      (GameplayTasks|EndTask (Variables|Transient|GetAsyncTask_RangeFiringAdd))
      (CallFunction|SetWeaponAttackingEnabled false)
      (Utilities|Time|ClearandInvalidateTimerbyHandle (Variables|Transient|GetPerformBeamVisualTraceDelayTimerHandle))
      (Utilities|Time|ClearandInvalidateTimerbyHandle (Variables|Transient|GetDisableWeaponFiringNextFrameTimerHandle)))
    (:then_1
      (|CallParentFunction bWasCancelled))))

(fn ParseTargetData
  (params (TargetData :type Gameplay%20Ability%20Target%20Data%20Handle%20Structure%20%28by%20ref%29))
  ;[0.05729 0.01279 0.05651 0.5] ParseTargetData: Break down the TargetData into a meaningful HitResult to use for the beam.
  (Utilities|FlowControl|Sequence
    (:then_0
      (if (Ability|TargetData|TargetDataHasHitResult TargetData)
        (return true (Ability|TargetData|GetHitResultfromTargetData TargetData))
        (else
          (bind _weapon (Utilities|GetThermalLanceWeapon)
            (:Valid
              (return true (Collision|MakeHitResult false false 0.0 0.0 "0, 0, 0" "0, 0, 0" "0,0,1" "0,0,1" 0 0 0 "None" "None" 0 0 0 (.ReturnValue (FortWeapon|GetDamageStartLocation _weapon -1)) (+ (.ReturnValue (FortWeapon|GetDamageStartLocation _weapon -1)) (* (FortWeapon|GetDamageStartLocation _weapon -1) (FortWeapon|GetRange _weapon))))))))))
    (:then_1
      (return false 0))))

(fn SetWeaponAttackingToDisabled
  (CallFunction|SetWeaponAttackingEnabled false))
