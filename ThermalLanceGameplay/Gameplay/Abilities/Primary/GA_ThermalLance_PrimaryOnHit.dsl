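; Gameplay ability for the Thermal Lance that passes the primary fire's trace hit result to the weapon's beam, invoked on server and client.
; On event activation it stores the instigator pawn, casts its current weapon to B_Weapon_ThermalLance and calls SetBeamInfo with the hit from the event's target data.
; It logs when hit results are missing or multiple, handling only the first, then ends itself via a Sequence.
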(event Ability|EventActivateAbilityFromEvent
  (params (EventData :type Gameplay%20Event%20Data%20Structure%20%28by%20ref%29))
  (bind (_eventtag _instigator _target _optionalobject _optionalobject2 _contexthandle _instigatortags _targettags _eventmagnitude _targetdata) (Utilities|Struct|BreakGameplayEventData EventData))
  ;[0.009853 0.15 0.01185 0.5] ActivateAbilityFromEvent: Ability is invoked whenever the weapon is fired. It is invoked both on server and client with the hit result of the primary attack's trace.
  (|CallParentFunction EventData)
  (Utilities|FlowControl|Sequence
    (:then_0
      (bind _asfort_player_pawn (Utilities|Casting|CastToFortPlayerPawn _instigator))
      (Variables|Transient|SetInstigatorPlayerPawn _asfort_player_pawn)
      (if (Ability|TargetData|TargetDataHasHitResult _targetdata)
        (if (> (Ability|TargetData|GetDataCountfromTargetData _targetdata) 1)
          (Fortnite|FortShippingLog self "ActivateAbility: Found more than 1 hit result but handling only the first hit result." true)
          (else
            _))
        (bind _asb_weapon_thermal_lance (Utilities|Casting|CastToB_Weapon_ThermalLance (Weapon|GetCurrentWeapon (Variables|Transient|GetInstigatorPlayerPawn))))
        (Beam|SetBeamInfo _asb_weapon_thermal_lance (Ability|TargetData|GetHitResultfromTargetData _targetdata))
        (else
          (Fortnite|FortShippingLog self "ActivateAbility: No HitResult found. Expecting 1." true))))
    (:then_1
      (Ability|EndAbility))))
